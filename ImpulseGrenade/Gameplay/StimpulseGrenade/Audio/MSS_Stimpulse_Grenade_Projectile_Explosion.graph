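# Stereo one-shot MetaSound Source for the Stimpulse Grenade projectile explosion, a preset of the shared MSS_Projectile_Explosion graph.
# It supplies three close and three distant Impulse Grenade waves, with inputs for layer volumes, distance curves, distant air-absorption low-pass and affiliation.
# The parent graph holds the logic, and the NoAtt attenuation asset paired with the distance inputs suggests falloff happens in-graph.
# It plays in the Explosives sound class, capped at three concurrent instances.
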MetaSoundSource /ImpulseGrenade/Gameplay/StimpulseGrenade/Audio/MSS_Stimpulse_Grenade_Projectile_Explosion
preset_parent=/FortMetaSoundWeapon/Items/Metasounds/Projectiles/MSS_Projectile_Explosion
inputs [
 Affiliation.Enemy {type=Float}
 Affiliation.Friendly {type=Float}
 Affiliation.Local {type=Float}
 Character.Affiliation {type=Enum:FortAudioAffiliation}
 Character.Health {type=Float}
 Character.UniqueID {type=Int32}
 DistantAirAbsorption.LowPassMax {type=Float}
 DistantAirAbsorption.LowPassMin {type=Float}
 DistantAirAbsorption.MaxDistance {type=Float}
 DistantAirAbsorption.MinDistance {type=Float}
 Explosion.CloseVolume {type=Float}
 Explosion.DistanceCurves {type=WaveTableBankAsset}
 Explosion.DistantVolume {type=Float}
 Explosion.FalloffDistance {type=Float}
 Explosion.InnerRadius {type=Float}
 Explosion.OutputVolume {type=Float}
 "Random.Max Pitch (Semitones)" {type=Float}
 Random.MinPitch {type=Float}
 Random.VolumeMax {type=Float}
 Random.VolumeMin {type=Float}
 Sounds.Close {type=WaveAsset:Array}
 Sounds.Distant {type=WaveAsset:Array}
 UE.Attenuation.Distance {type=Float}
 UE.Source.OnPlay {type=Trigger}
]
interfaces [Fort.Character:1.0, UE.Attenuation:1.0, UE.OutputFormat.Stereo:1.0, UE.Source.OneShot:1.0, UE.Source:1.1]
outputs [
 "UE.OutputFormat.Stereo.Audio:0" {type=Audio}
 "UE.OutputFormat.Stereo.Audio:1" {type=Audio}
 UE.Source.OneShot.OnFinished {type=Trigger}
]
pages [Default {input_defaults {
   Affiliation.Enemy=0
   Affiliation.Friendly=-6
   Affiliation.Local=0
   Character.Affiliation=0
   Character.Health=0
   Character.UniqueID=0
   DistantAirAbsorption.LowPassMax=6000
   DistantAirAbsorption.LowPassMin=20000
   DistantAirAbsorption.MaxDistance=27500
   DistantAirAbsorption.MinDistance=8000
   Explosion.CloseVolume=0
   Explosion.DistanceCurves=/FortMetaSoundWeapon/Items/Shared/Settings/Projectiles/WTB_Projectile_Explosion_Distance
   Explosion.DistantVolume=-12
   Explosion.FalloffDistance=27500
   Explosion.InnerRadius=0
   Explosion.OutputVolume=0
   "Random.Max Pitch (Semitones)"=0
   Random.MinPitch=0
   Random.VolumeMax=0
   Random.VolumeMin=0
   UE.Attenuation.Distance=14519.9922
   UE.Source.OnPlay=null
   Sounds.Close [
    /ImpulseGrenade/Gameplay/StimpulseGrenade/Audio/Wav/Impulse_Grenade_Explode_Close_01
    /ImpulseGrenade/Gameplay/StimpulseGrenade/Audio/Wav/Impulse_Grenade_Explode_Close_02
    /ImpulseGrenade/Gameplay/StimpulseGrenade/Audio/Wav/Impulse_Grenade_Explode_Close_03
   ]
   Sounds.Distant [
    /ImpulseGrenade/Gameplay/StimpulseGrenade/Audio/Wav/Impulse_Grenade_Explode_Distant_01
    /ImpulseGrenade/Gameplay/StimpulseGrenade/Audio/Wav/Impulse_Grenade_Explode_Distant_02
    /ImpulseGrenade/Gameplay/StimpulseGrenade/Audio/Wav/Impulse_Grenade_Explode_Distant_03
   ]
  }}]
properties {
 assetImportData=&self:AssetImportData
 attenuationSettings=/FortMetaSoundWeapon/Items/Shared/Settings/Projectiles/ATT_Projectile_Explosion_NoAtt
 bOverrideConcurrency=true
 maxDistance=27500
 numChannels=2
 outputFormat=Stereo
 sampleRate=48000
 soundClassObject=/Game/Sounds/FortSoundClassesAndModes/Fort_Explosives
 concurrencyOverrides {maxCount=3}
}
